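# Niagara system for the Gumball emote: a short, burst-spawned pink splat effect placed by sampling the User.SkeletalMesh.
# Splat and Warp sprites expand and fade, with Warp also animating a dynamic material parameter.
# Drips throws small SubUV sprites that fall and shrink.
# Eyes launches two growing mesh particles, using spawn groups to mirror their side offsets, with camera offset and animated material parameters.
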NiagaraSystem /BRCosmetics/Animation/Game/MainPlayer/Emotes/Gumball/NS_Gumball_Emote
dependencies {
 usedDataInterfaces [&NiagaraDataInterfaceColorCurve, &NiagaraDataInterfaceCurve, &NiagaraDataInterfaceVector2DCurve]
 usedDynamicInputs [&Vector2DFromFloat]
 usedModules [&SystemState]
}
emitters {
 Drips {
  emitterUpdateScript [
   &EmitterState()
   &SpawnRate(SpawnRate=5, Use Spawn Probability=true) {enabled=false}
   &SpawnBurst_Instantaneous(Spawn Count=3, Use Spawn Probability=false, UseLoopCountLimit=false)
  ]
  particleSpawnScript [
   &InitializeParticle(
    Lifetime=&RandomRangeFloat(Minimum=0.4, Maximum=0.6)
    Sprite Rotation Mode=::NewEnumerator2
    Sprite Size Mode=::NewEnumerator3
    Uniform Sprite Size=&RandomRangeFloat(Minimum=10, Maximum=20)
    UsePositionOffset=false
    Write Material Random=true
   )
   &SkeletalMeshLocation(
    MaskSpawnGroup=false
    Mesh Sampling Type=::NewEnumerator5
    Position Sampling=::NewEnumerator1
    Skeletal Mesh=$User.SkeletalMesh
    Socket Sampling Mode=::NewEnumerator5
   )
   &SubUVAnimation(End Frame=3, SubUV Animation Mode:ENiagara_SubUVLookupMode=::NewEnumerator2)
   &AddVelocityFromPoint(Module.Use Falloff Distance=false, Velocity Strength=200)
  ]
  particleUpdateScript [
   &ParticleState()
   &Drag(Use Linear Drag=true, Use Rotational Drag=false)
   &GravityForce()
   &SolveForcesAndVelocity(Clamp Velocity=false, Limit Acceleration=false)
   &ScaleColor(
    Scale Alpha=&FloatFromCurve(FloatCurve=DataInterface() {propertyValues {Curve {keys [
         {arriveTangent=-2.5249, interpMode=RCIM_Cubic, time=0.1, value=1}
         {arriveTangent=-0.9941, interpMode=RCIM_Cubic, leaveTangent=-0.9941, tangentMode=RCTM_User, time=0.4, value=0.5}
         {arriveTangent=-1.0623, interpMode=RCIM_Cubic, leaveTangent=-1.0623, tangentMode=RCTM_User, time=1, value=0}
        ]}}})
    ScaleA=true
    ScaleRGB=false
   ) {enabled=false}
   &ScaleSpriteSize(Uniform Curve Sprite Scale=DataInterface() {propertyValues {Curve {keys [{interpMode=RCIM_Cubic, tangentMode=RCTM_User, time=0, value=1}, {interpMode=RCIM_Cubic, tangentMode=RCTM_User, time=1, value=0}]}}})
  ]
  rendererClasses [&NiagaraSpriteRendererProperties]
  renderers [{rendererClass=&NiagaraSpriteRendererProperties}]
 }
 Eyes {
  emitterUpdateScript [
   &EmitterState()
   &SpawnBurst_Instantaneous(Spawn Count=2, Use Spawn Probability=false, UseLoopCountLimit=false)
  ]
  particleSpawnScript [
   &FN_SetSpawnGroup(Number of Spawn Groups=2)
   &InitializeParticle(
    Lifetime=&RandomRangeFloat(Minimum=0.2, Maximum=0.4)
    Mesh Scale Mode=::NewEnumerator3
    Mesh Uniform Scale=8
    Position Offset=&MaskVectorBySpawnGroup(Masked Vector=<-7 0 15>, Passthrough Vector=<7 0 15>)
    UsePositionOffset=false
    Write Material Random=true
   )
   &InitialMeshOrientation(Use Rotation Vector=false)
   &SkeletalMeshLocation(
    MaskSpawnGroup=false
    Mesh Sampling Type=::NewEnumerator5
    Position Sampling=::NewEnumerator1
    Sampled Position Offset=&MaskVectorBySpawnGroup(Masked Vector=<10 4 4>, Passthrough Vector=<-10 4 4>)
    Skeletal Mesh=$User.SkeletalMesh
    Socket Sampling Mode=::NewEnumerator5
   )
   &AddVelocityFromPoint(Default Position:NiagaraPosition=$Particles.SkeletalMeshLocation.SampledPosition, Module.Use Falloff Distance=false, Velocity Strength=170)
  ]
  particleUpdateScript [
   &ParticleState()
   &Drag(Use Linear Drag=true, Use Rotational Drag=false)
   &GravityForce()
   &SolveForcesAndVelocity(Clamp Velocity=false, Limit Acceleration=false)
   &DynamicMaterialParameters(
    Index 0 Param 1=&RandomRangeFloat(Minimum=0, Maximum=1)
    Index 0 Param 2=1
    Index 0 Param 3=2
    Index 0 Param 4=&FloatFromCurve(FloatCurve=DataInterface() {propertyValues {Curve {keys [
         {arriveTangent=-3.7762, interpMode=RCIM_Cubic, leaveTangent=-3.7762, tangentMode=RCTM_User, time=0, value=1}
         {arriveTangent=-1.2694, interpMode=RCIM_Cubic, leaveTangent=-1.2694, tangentMode=RCTM_User, time=1, value=-1}
        ]}}})
    Index 1 Param 1=0.3
    Index 1 Param 2=0.3
    Module.Param0WriteEnabled=true
    Module.Param0WriteEnabled1=true
    Module.Param1WriteEnabled=true
    Module.Param1WriteEnabled1=true
    Module.Param2WriteEnabled=true
    Module.Param2WriteEnabled1=true
    Module.Param3WriteEnabled=true
    Module.Param3WriteEnabled1=true
    Write Parameter Index 1=true
   )
   &ScaleMeshSize(Scale Factor=&VectorFromFloat(Value=&FloatFromCurve(FloatCurve=DataInterface() {propertyValues {Curve {keys [
          {arriveTangent=1.0798, interpMode=RCIM_Cubic, leaveTangent=1.0798, tangentMode=RCTM_User, time=0, value=1}
          {interpMode=RCIM_Cubic, tangentMode=RCTM_User, time=1, value=3}
         ]}}})))
   &CameraOffset(Camera Offset Amount=20)
  ]
  rendererClasses [&NiagaraMeshRendererProperties]
  renderers [{rendererClass=&NiagaraMeshRendererProperties, rendererIndex=0}]
 }
 Splat {
  emitterUpdateScript [
   &EmitterState()
   &SpawnRate(SpawnRate=5, Use Spawn Probability=true) {enabled=false}
   &SpawnBurst_Instantaneous(Use Spawn Probability=false, UseLoopCountLimit=false)
  ]
  particleSpawnScript [
   &InitializeParticle(
    Color=<rgba 1 0.203 0.4846 1>
    Lifetime=0.45
    Sprite Rotation Mode=::NewEnumerator2
    Sprite Size Mode=::NewEnumerator3
    Uniform Sprite Size=100
    UsePositionOffset=false
    Write Material Random=true
   )
   /SetVariables(Particles.Position=&Add_Vector(
     A=/Niagara/DynamicInputs/Helpers/SimulationPosition()
     B=&MakeVector(
      X=&FloatFromCurve(FloatCurve=DataInterface() {propertyValues {Curve {keys [<at 0.5 1>,<at 3 0>]}}}, CurveIndex=$System.Age, Scale Curve=7)
      Y=0
      Z=0
     )
    )) {enabled=false, isSetParameters=true}
   &SkeletalMeshLocation(
    MaskSpawnGroup=false
    Mesh Sampling Type=::NewEnumerator5
    Position Sampling=::NewEnumerator1
    Skeletal Mesh=$User.SkeletalMesh
    Socket Sampling Mode=::NewEnumerator5
   )
  ]
  particleUpdateScript [
   &ParticleState()
   &GravityForce()
   &SolveForcesAndVelocity(Clamp Velocity=false, Limit Acceleration=false)
   &ScaleColor(
    Scale Alpha=&FloatFromCurve(FloatCurve=DataInterface() {propertyValues {Curve {keys [
         {arriveTangent=-1.1111, interpMode=RCIM_Cubic, leaveTangent=-1.1111, time=0.1, value=1}
         {arriveTangent=-0.9941, interpMode=RCIM_Cubic, leaveTangent=-0.9941, tangentMode=RCTM_User, time=0.4, value=0.5}
         {arriveTangent=-1.0623, interpMode=RCIM_Cubic, leaveTangent=-1.0623, tangentMode=RCTM_User, time=1, value=0}
        ]}}})
    ScaleA=true
    ScaleRGB=false
   )
   &ScaleSpriteSize(
    Uniform Curve Scale=1.2
    Uniform Curve Sprite Scale=DataInterface() {propertyValues {Curve {keys [
        {arriveTangent=1.2875, interpMode=RCIM_Cubic, leaveTangent=1.2875, tangentMode=RCTM_User, time=0, value=0.3}
        {interpMode=RCIM_Cubic, tangentMode=RCTM_User, time=1, value=1}
       ]}}}
   )
   &DynamicMaterialParameters(
    Module.Param0WriteEnabled1=true
    Module.Param1WriteEnabled1=true
    Module.Param2WriteEnabled1=true
    Module.Param3WriteEnabled1=true
    Write Parameter Index 0=false
    Write Parameter Index 1=true
   )
   &CameraOffset(Camera Offset Amount=10)
  ]
  rendererClasses [&NiagaraSpriteRendererProperties]
  renderers [{rendererClass=&NiagaraSpriteRendererProperties}]
 }
 Warp {
  emitterUpdateScript [
   &EmitterState()
   &SpawnRate(SpawnRate=5, Use Spawn Probability=true) {enabled=false}
   &SpawnBurst_Instantaneous(Use Spawn Probability=false, UseLoopCountLimit=false)
  ]
  particleSpawnScript [
   &InitializeParticle(
    Color=<rgba 1 0.203 0.4846 1>
    Lifetime=0.4
    Sprite Rotation Mode=::NewEnumerator2
    Sprite Size Mode=::NewEnumerator3
    Uniform Sprite Size=110
    UsePositionOffset=false
    Write Material Random=true
   )
   &SkeletalMeshLocation(
    MaskSpawnGroup=false
    Mesh Sampling Type=::NewEnumerator5
    Position Sampling=::NewEnumerator1
    Skeletal Mesh=$User.SkeletalMesh
    Socket Sampling Mode=::NewEnumerator5
   )
  ]
  particleUpdateScript [
   &ParticleState()
   &SolveForcesAndVelocity(Clamp Velocity=false, Limit Acceleration=false)
   &ScaleColor(
    Scale Alpha=&FloatFromCurve(FloatCurve=DataInterface() {propertyValues {Curve {keys [
         {arriveTangent=-2.5249, interpMode=RCIM_Cubic, time=0.1, value=1}
         {arriveTangent=-0.9941, interpMode=RCIM_Cubic, leaveTangent=-0.9941, tangentMode=RCTM_User, time=0.4, value=0.5}
         {arriveTangent=-1.0623, interpMode=RCIM_Cubic, leaveTangent=-1.0623, tangentMode=RCTM_User, time=1, value=0}
        ]}}})
    ScaleA=true
    ScaleRGB=false
   )
   &ScaleSpriteSize(Uniform Curve Sprite Scale=DataInterface() {propertyValues {Curve {keys [
        {arriveTangent=1.2875, interpMode=RCIM_Cubic, leaveTangent=1.2875, tangentMode=RCTM_User, time=0, value=0.3}
        {interpMode=RCIM_Cubic, tangentMode=RCTM_User, time=1, value=1}
       ]}}})
   &DynamicMaterialParameters(
    Index 1 Param 1=&FloatFromCurve(
     FloatCurve=DataInterface() {propertyValues {Curve {keys [{interpMode=RCIM_Cubic, tangentMode=RCTM_User, time=0, value=0}, {interpMode=RCIM_Cubic, tangentMode=RCTM_User, time=1, value=1}]}}}
     Scale Curve=5
    )
    Module.Param0WriteEnabled1=true
    Module.Param1WriteEnabled1=true
    Module.Param2WriteEnabled1=true
    Module.Param3WriteEnabled1=true
    Write Parameter Index 0=false
    Write Parameter Index 1=true
   )
  ]
  rendererClasses [&NiagaraSpriteRendererProperties]
  renderers [{rendererClass=&NiagaraSpriteRendererProperties}]
 }
}
summary {
 systemName=NS_Gumball_Emote
 emitters [
  {emitterName=Warp, rendererClasses [&NiagaraSpriteRendererProperties]}
  {emitterName=Splat, rendererClasses [&NiagaraSpriteRendererProperties]}
  {emitterName=Eyes, rendererClasses [&NiagaraMeshRendererProperties]}
  {emitterName=Drips, rendererClasses [&NiagaraSpriteRendererProperties]}
 ]
 userVariables [
  User.bIsFrontEnd {type=NiagaraBool}
  User.bIsFrontEndPreview {type=NiagaraBool}
  User.bIsFullyDeployed {type=NiagaraBool}
  User.ForwardDot {type=NiagaraFloat, defaultValue=0, description=""}
  User.SkeletalMesh {
   type=NiagaraDataInterfaceSkeletalMesh
   defaultValue {
    struct=NiagaraExt_VariableValue_DataInterface
    value {dataInterface=&self:NiagaraDataInterfaceSkeletalMesh_0}
   }
  }
 ]
}
